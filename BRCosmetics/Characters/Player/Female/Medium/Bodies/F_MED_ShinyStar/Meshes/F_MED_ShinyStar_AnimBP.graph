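# Animation Blueprint for the F_MED_ShinyStar female body cosmetic, driving its mesh from the character mesh it is attached to.
# It copies the attached parent's pose and feeds it through the ShinyStar Control Rig.
# An additive bone-space Modify Bone then nudges the left pouch bone dyn_pouch_l_01, enabled by bIsCrouching, which is not declared here.
# The Event Graph holds only an unconnected update event.
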AnimBlueprint /BRCosmetics/Characters/Player/Female/Medium/Bodies/F_MED_ShinyStar/Meshes/F_MED_ShinyStar_AnimBP
target_skeleton=&/F_MED_ShinyStar_Skeleton
graphs [
 AnimGraph {nodes [
   n1 = Control Rig(Source=Copy Pose From Mesh() {bUseAttachedParent=true}) {
    alphaBoolBlend {blendOption=Linear}
    controlRigAssetReference {blueprintRigClass=&/F_MED_ShinyStar_ControlRig.F_MED_ShinyStar_ControlRig_C}
   }
   n2 = "Transform (Modify) Bone - Bone: dyn_pouch_l_01"!(
    ComponentPose=Local To Component(LocalPose=@n1)
    bAlphaBoolEnabled:bool=Get bIsCrouching()
    Translation="0,-1,0"
    Rotation="0,-5,10"
    Scale="1,1,1"
   ) -> (Pose) {
    alphaInputType=Bool
    rotationMode=BMM_Additive
    rotationSpace=BCS_BoneSpace
    translationMode=BMM_Additive
    translationSpace=BCS_BoneSpace
    alphaBoolBlend {blendOption=Linear}
    boneToModify {boneName=dyn_pouch_l_01}
    rotation {roll=10, yaw=-5}
    translation {y=-1}
   }
   Output Pose(Result=Component To Local(ComponentPose=@n2)) {layerGroup=None}
  ]}
 EventGraph {nodes [Event BlueprintUpdateAnimation()]}
]
variables []
