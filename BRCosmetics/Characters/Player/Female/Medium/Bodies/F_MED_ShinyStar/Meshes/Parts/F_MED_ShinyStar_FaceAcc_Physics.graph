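# Physics asset for the female medium ShinyStar face accessory part, defining capsule bodies and joints for dynamic hair and a right-side strap.
# Spine, neck and head capsules form the base, with hair and strap master bones attached to the head by free-rotation constraints.
# A four-bone center chain, two left chains sharing one master, and short right hair and strap segments hang from those masters with narrow swing limits and zero twist.
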PhysicsAsset /BRCosmetics/Characters/Player/Female/Medium/Bodies/F_MED_ShinyStar/Meshes/Parts/F_MED_ShinyStar_FaceAcc_Physics
bodies {
 dyn_hairMeshA_l_01 [{shapeType=Capsule, length=2, radius=1, shapeName=dyn_hairMeshA_l_01_capsule, center=<2.0174 0 0>, rotation=<rot 90 0 0>}]
 dyn_hairMeshA_l_02 [{shapeType=Capsule, length=2, radius=1, shapeName=dyn_hairMeshA_l_02_capsule, center=<2.0174 0 0>, rotation=<rot 90 0 0>}]
 dyn_hairMeshA_l_03 [{shapeType=Capsule, length=2, radius=1, shapeName=dyn_hairMeshA_l_03_capsule, center=<2.0174 0 0>, rotation=<rot 90 0 0>}]
 dyn_hairMeshA_l_MSTR [{shapeType=Capsule, length=0.5, radius=0.5, shapeName=dyn_hairMeshA_l_MSTR_capsule}]
 dyn_hairMeshB_l_01 [{shapeType=Capsule, length=2, radius=1, shapeName=dyn_hairMeshB_l_01_capsule, center=<2.0174 0 0>, rotation=<rot 90 0 0>}]
 dyn_hairMeshB_l_02 [{shapeType=Capsule, length=2, radius=1, shapeName=dyn_hairMeshB_l_02_capsule, center=<2.0174 0 0>, rotation=<rot 90 0 0>}]
 dyn_hairMeshB_l_03 [{shapeType=Capsule, length=2, radius=1, shapeName=dyn_hairMeshB_l_03_capsule, center=<2.0174 0 0>, rotation=<rot 90 0 0>}]
 dyn_hairMesh_r_01 [{shapeType=Capsule, length=2, radius=1, shapeName=dyn_hairMesh_r_01_capsule, center=<2.0174 0 1.3>, rotation=<rot 90 0 0>}]
 dyn_hairMesh_r_MSTR [{shapeType=Capsule, length=0.5, radius=0.5, shapeName=dyn_hairMesh_r_MSTR_capsule}]
 dyn_hair_c_01 [{shapeType=Capsule, length=0.945, radius=2.6914, shapeName=dyn_hair_c_01_capsule, center=<3.1803 0 0>, rotation=<rot 90 0 0>}]
 dyn_hair_c_02 [{shapeType=Capsule, length=0, radius=3.4684, shapeName=dyn_hair_c_02_capsule, center=<3.562 0 0>, rotation=<rot 90 0 0>}]
 dyn_hair_c_03 [{
   shapeType=Capsule
   length=4.7321
   radius=3.1567
   shapeName=dyn_hair_c_03_capsule
   center=<3.3268 -1.4 0>
   rotation=<rot 90 -175.2364 -84.4853>
  }]
 dyn_hair_c_04 [{shapeType=Capsule, length=3.847, radius=2.8439, shapeName=dyn_hair_c_04_capsule, center=<4.0449 0 0>, rotation=<rot 90 0 0>}]
 dyn_hair_c_MSTR [{shapeType=Capsule, length=0.5, radius=0.5, shapeName=dyn_hair_c_MSTR_capsule}]
 dyn_strap_r_01 [{shapeType=Capsule, length=2.009, radius=0.991, shapeName=dyn_strap_r_01_capsule, rotation=<rot 90 0 0>}]
 dyn_strap_r_MSTR [{shapeType=Capsule, length=0.5, radius=0.5, shapeName=dyn_strap_r_MSTR_capsule}]
 head [{shapeType=Capsule, length=6.3837, radius=7.3076, shapeName=head_capsule, center=<1.9617 1.8502 -0.4479>, rotation=<rot 90 0 0>}]
 neck_01 [{shapeType=Capsule, length=3.1643, radius=4.8472, shapeName=neck_01_capsule, rotation=<rot 90 0 0>}]
 neck_02 [{shapeType=Capsule, length=3.1643, radius=4.8472, shapeName=neck_02_capsule, rotation=<rot 90 0 0>}]
 spine_05 [{shapeType=Capsule, length=15.4018, radius=10.8309, shapeName=spine_05_capsule, center=<2.1805 0 0>, rotation=<rot 0 -90 0>}]
}
constraints [
 {bone1Name=dyn_strap_r_01, bone2Name=dyn_strap_r_MSTR, swing1LimitDegrees=10, swing2LimitDegrees=0, twistLimitDegrees=0}
 {
  bone1Name=dyn_hairMeshA_l_MSTR
  bone2Name=head
  swing1LimitDegrees=45
  swing1Motion=Free
  swing2LimitDegrees=45
  swing2Motion=Free
  twistLimitDegrees=45
  twistMotion=Free
 }
 {bone1Name=dyn_hairMeshA_l_01, bone2Name=dyn_hairMeshA_l_MSTR, swing1LimitDegrees=10, swing2LimitDegrees=10, twistLimitDegrees=0}
 {bone1Name=dyn_hairMeshA_l_02, bone2Name=dyn_hairMeshA_l_01, swing1LimitDegrees=10, swing2LimitDegrees=10, twistLimitDegrees=0}
 {bone1Name=dyn_hairMeshA_l_03, bone2Name=dyn_hairMeshA_l_02, swing1LimitDegrees=20, swing2LimitDegrees=20, twistLimitDegrees=0}
 {bone1Name=dyn_hairMeshB_l_01, bone2Name=dyn_hairMeshA_l_MSTR, swing1LimitDegrees=10, swing2LimitDegrees=10, twistLimitDegrees=0}
 {bone1Name=dyn_hairMeshB_l_02, bone2Name=dyn_hairMeshB_l_01, swing1LimitDegrees=10, swing2LimitDegrees=10, twistLimitDegrees=0}
 {bone1Name=dyn_hairMeshB_l_03, bone2Name=dyn_hairMeshB_l_02, swing1LimitDegrees=20, swing2LimitDegrees=20, twistLimitDegrees=0}
 {
  bone1Name=dyn_hairMesh_r_MSTR
  bone2Name=head
  swing1LimitDegrees=45
  swing1Motion=Free
  swing2LimitDegrees=45
  swing2Motion=Free
  twistLimitDegrees=45
  twistMotion=Free
 }
 {bone1Name=dyn_hairMesh_r_01, bone2Name=dyn_hairMesh_r_MSTR, swing1LimitDegrees=5, swing2LimitDegrees=5, twistLimitDegrees=0}
 {
  bone1Name=dyn_hair_c_MSTR
  bone2Name=head
  swing1LimitDegrees=45
  swing1Motion=Free
  swing2LimitDegrees=45
  swing2Motion=Free
  twistLimitDegrees=45
  twistMotion=Free
 }
 {bone1Name=dyn_hair_c_01, bone2Name=dyn_hair_c_MSTR, swing1LimitDegrees=25, swing2LimitDegrees=30, twistLimitDegrees=0}
 {bone1Name=dyn_hair_c_02, bone2Name=dyn_hair_c_01, swing1LimitDegrees=30, swing2LimitDegrees=30, twistLimitDegrees=0}
 {bone1Name=dyn_hair_c_03, bone2Name=dyn_hair_c_02, swing1LimitDegrees=30, swing2LimitDegrees=20, twistLimitDegrees=0}
 {bone1Name=dyn_hair_c_04, bone2Name=dyn_hair_c_03, swing1LimitDegrees=20, swing2LimitDegrees=20, twistLimitDegrees=0}
 {
  bone1Name=dyn_strap_r_MSTR
  bone2Name=head
  swing1LimitDegrees=45
  swing1Motion=Free
  swing2LimitDegrees=45
  swing2Motion=Free
  twistLimitDegrees=45
  twistMotion=Free
 }
 {
  bone1Name=head
  bone2Name=neck_02
  swing1LimitDegrees=45
  swing1Motion=Free
  swing2LimitDegrees=45
  swing2Motion=Free
  twistLimitDegrees=45
  twistMotion=Free
 }
 {
  bone1Name=neck_01
  bone2Name=spine_05
  swing1LimitDegrees=45
  swing1Motion=Free
  swing2LimitDegrees=45
  swing2Motion=Free
  twistLimitDegrees=45
  twistMotion=Free
 }
 {
  bone1Name=neck_02
  bone2Name=neck_01
  swing1LimitDegrees=45
  swing1Motion=Free
  swing2LimitDegrees=45
  swing2Motion=Free
  twistLimitDegrees=45
  twistMotion=Free
 }
]
