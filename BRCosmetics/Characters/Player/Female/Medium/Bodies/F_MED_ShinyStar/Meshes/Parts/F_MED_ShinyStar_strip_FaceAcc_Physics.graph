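# Physics asset for the face accessory of the female medium ShinyStar character body, covering a right-side strap.
# Three capsule bodies cover the head, the strap master bone and the first strap bone.
# Constraints chain them together: the master bone rotates freely on all angular axes relative to the head, while the strap bone is tightly restricted to a small single-axis swing against the master.
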PhysicsAsset /BRCosmetics/Characters/Player/Female/Medium/Bodies/F_MED_ShinyStar/Meshes/Parts/F_MED_ShinyStar_strip_FaceAcc_Physics
bodies {
 dyn_strap_r_01 [{shapeType=Capsule, length=2.009, radius=0.991, shapeName=dyn_strap_r_01_capsule, center=<0.6662 0 0>, rotation=<rot 90 0 0>}]
 dyn_strap_r_MSTR [{shapeType=Capsule, length=0.5, radius=0.5, shapeName=dyn_strap_r_MSTR_capsule}]
 head [{shapeType=Capsule, length=6.3837, radius=7.3076, shapeName=head_capsule, center=<1.9617 1.8502 -0.4479>, rotation=<rot 90 0 0>}]
}
constraints [
 {bone1Name=dyn_strap_r_01, bone2Name=dyn_strap_r_MSTR, swing1LimitDegrees=10, swing2LimitDegrees=0, twistLimitDegrees=0}
 {
  bone1Name=dyn_strap_r_MSTR
  bone2Name=head
  swing1LimitDegrees=45
  swing1Motion=Free
  swing2LimitDegrees=45
  swing2Motion=Free
  twistLimitDegrees=45
  twistMotion=Free
 }
]
