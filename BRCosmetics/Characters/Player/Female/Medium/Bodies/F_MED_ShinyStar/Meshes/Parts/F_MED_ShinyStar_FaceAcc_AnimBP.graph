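# Animation Blueprint for the ShinyStar female skin's face accessory, adding secondary motion on top of the pose copied from the attached parent mesh.
# A cosmetic variant check picks jacket or standard RigidBody physics, feeding a head-relative strip simulation, all using a gravity override.
# When the parent's FrontEnd_Idle_Bypass curve is active, physics gives way to a lobby hair sequence, layered from dyn_hair_c_MSTR and time-synced to the idle montage.
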AnimBlueprint /BRCosmetics/Characters/Player/Female/Medium/Bodies/F_MED_ShinyStar/Meshes/Parts/F_MED_ShinyStar_FaceAcc_AnimBP
target_skeleton=&/F_MED_ShinyStar_FaceAcc_Skeleton
graphs [
 AnimGraph {nodes [
   n1 = RigidBody!(
    ComponentPose=Local To Component(LocalPose=Copy Pose From Mesh() {bUseAttachedParent=true})
    Alpha
    OverrideWorldGravity:Vector=Get Gravity Override()
    ExternalForce
   ) -> (Pose) {
    bClampLinearTranslationLimitToRefPose=true
    bOverrideWorldGravity=true
    overridePhysicsAsset=&/F_MED_ShinyStarJacket_FaceAcc_Physics
    simulationSpace=BaseBoneSpace
    alphaBoolBlend {blendOption=Linear}
    baseBoneRef {boneName=root}
    componentAppliedLinearAccClamp=<6000 6000 6000>
    componentLinearAccScale=<0.45 0.2 0.3>
    componentLinearVelScale=<0.45 0.2 0.3>
    overrideWorldGravity {z=-980}
   }
   n2 = Component To Local(ComponentPose=@n1)
   n3 = RigidBody(ComponentPose=Local To Component(LocalPose=Copy Pose From Mesh() {bUseAttachedParent=true}), OverrideWorldGravity=Get Gravity Override()) {
    bClampLinearTranslationLimitToRefPose=true
    bOverrideWorldGravity=true
    overridePhysicsAsset=&/F_MED_ShinyStar_FaceAcc_Physics
    simulationSpace=BaseBoneSpace
    alphaBoolBlend {blendOption=Linear}
    baseBoneRef {boneName=root}
    componentAppliedLinearAccClamp=<6000 6000 6000>
    componentLinearAccScale=<0.45 0.2 0.3>
    componentLinearVelScale=<0.45 0.2 0.3>
    overrideWorldGravity {z=-980}
   }
   n4 = Component To Local(ComponentPose=@n3)
   n5 = Get isJacketVariant() -> (isJacketVariant:bool)
   n6 = RigidBody(
    ComponentPose=Local To Component(LocalPose=Blend Poses by bool(BlendPose_0=@n2, BlendPose_1=@n4, bActiveValue=@n5) {blendType=Linear})
    OverrideWorldGravity=Get Gravity Override()
   ) {
    bClampLinearTranslationLimitToRefPose=true
    bOverrideWorldGravity=true
    overridePhysicsAsset=&/F_MED_ShinyStar_strip_FaceAcc_Physics
    simulationSpace=BaseBoneSpace
    alphaBoolBlend {blendOption=Linear}
    baseBoneRef {boneName=head}
    componentAppliedLinearAccClamp=<6000 6000 6000>
    componentLinearAccScale=<0.6 0.2 0.2>
    componentLinearVelScale=<0.6 0.2 0.2>
    overrideWorldGravity {z=-980}
   }
   n7 = Copy Pose From Mesh() {bUseAttachedParent=true}
   n8 = "Sequence Evaluator (by time) 'ShinyStar_FaceAccesory_LobbyIdle'"!(ExplicitTime:float=Get Hair Position() -> (Hair Position:double)) -> (Pose:Pose Link Structure) {
    is=title AnimGraphNode_SequenceEvaluator
    sequence=/BRCosmetics/Animation/Game/MainPlayer/Menu/BR/CustomSkins/_s23/ShinyStarF/ShinyStar_FaceAccesory_LobbyIdle
   }
   Output Pose(Result=Blend Poses by bool(
     BlendPose_0=Layered blend per bone(BasePose=@n7, BlendPoses_0=@n8) {layerSetup [{branchFilters [{boneName=dyn_hair_c_MSTR}]}]}
     BlendPose_1=Component To Local(ComponentPose=@n6)
     bActiveValue=Get PlayingFrontendIdle() -> (PlayingFrontendIdle:bool)
    ) {blendType=Linear}) {layerGroup=None}
  ]}
 EventGraph {nodes [
   Event BlueprintUpdateAnimation()
   > n1 = Sequence!(execute) -> (then_0, then_1, then_2:Exec, then_3:Exec)
   > n2 = Gravity Override State Function DEPRECATED(CharacterPartAnimInstance=Self-Reference(), GameTimeInSeconds=0, GravityOverrideParams=Get Gravity Override Params())
   > Set Gravity Override(Gravity Override=@n2)
   n3 = Cast To FortPlayerPawn(execute=@n1.then_1, Object=GetOwningActor())
   > n4 = GetVariantChannelActiveIndexForItemActor!(
    execute
    Item:Athena Cosmetic Item Definition Object Reference=/BRCosmetics/Athena/Items/Cosmetics/Characters/Character_ShinyStar
    Actor:Actor Object Reference=@n3
    Channel:Gameplay Tag Structure=(TagName="Cosmetics.Variant.Channel.Parts")
    VariantList:Array of Gameplay Tag Structures=Make Array("[0]"=(TagName="Cosmetics.Variant.Property.Stage2"))
    bErrorIfChannelMissing:bool=false
    bErrorIfTagNotInArray:bool=false
   ) -> (then, Index:int=0, bIndexIsValid:bool, bFoundChannel:bool)
   > n5 = Branch(Condition=AND Boolean(A="Equal (Integer)"(A=@n4), B=@n4.bIndexIsValid))
   > Set isJacketVariant(isJacketVariant:bool=true) -> (Output_Get:bool)
   Set isJacketVariant(execute=@n5.else, isJacketVariant=false)
   n6 = GetAnimInstance(self=GetOwnerSkeletalMeshComponent())
   Set PlayingFrontendIdle(
    execute=@n1.then_2
    PlayingFrontendIdle:bool="float > float"(A=GetCurveValue(self=@n6, CurveName=FrontEnd_Idle_Bypass), B=0) {is=title K2Node_CallFunction}
   ) -> (Output_Get:bool)
   n7 = GetOwnerSkeletalMeshComponent()
   Set Hair Position(
    execute=@n1.then_3
    Hair Position:double=Montage_GetPosition!(
     self:Anim Instance Object Reference=GetAnimInstance(self=@n7)
     Montage:Anim Montage Object Reference=/BRCosmetics/Animation/Game/MainPlayer/Menu/BR/CustomSkins/_s23/ShinyStarF/ShinyStar_Female_idle_Montage
    ) -> (ReturnValue:float)
   ) -> (Output_Get:double)
  ]}
]
variables [
 Anim_Inst {type=AnimInstance, default=null}
 Gravity Override {type=Vector}
 Gravity Override Params {type=GravityOverrideParamsStruct}
 Hair Position {type=number}
 isJacketVariant {type=boolean}
 PlayingFrontendIdle {type=boolean}
]
